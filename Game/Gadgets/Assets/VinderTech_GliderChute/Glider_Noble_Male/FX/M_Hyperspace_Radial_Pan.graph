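# Translucent material for Niagara mesh particles that draws a swirling, radially panning hyperspace cloud effect for the Noble Male glider.
# A sphere mask drives a UV rotator for an adjustable twist, then radial coordinates pan smoke and noise textures at different tilings and speeds.
# The blended noise, inverted and reduced by a center mask, indexes a four-color blue-based ramp that feeds emissive adjustments.
# Several masks, colors and DDX/DDY parameters are left unconnected.
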Material /Game/Gadgets/Assets/VinderTech_GliderChute/Glider_Noble_Male/FX/M_Hyperspace_Radial_Pan
graph {
 properties {
  bUsedWithNiagaraMeshParticles=true
  blendMode=BLEND_Translucent
  pixelDepthOffsetMode=PDOM_Legacy
  previewMesh=/Game/Effects/Fort_Effects/Mesh/Smoke/SM_Hemisphere_Concave
  usedShadingModels=Default Lit
 }
 nodes [
  n1 = SphereMask(A=TextureCoordinate(), B=Constant2Vector() {g=0.5, r=0.5}, Radius=Constant() {r=0.47}, Hardness=Constant())
  n2 = SphereMask(A=TextureCoordinate(), B=Constant2Vector() {g=0.5, r=0.5}, Radius=Constant() {r=0.2}, Hardness=Constant())
  Add(A=@n1, B=@n2)
  Constant3Vector() {constant=<rgba 1 1 1 0>}
  Constant3Vector() {constant=<rgba 1 0.1455 0.2753 1>}
  n3 = Constant3Vector() {constant=<rgba 0.0249 0.1983 1 1>}
  n4 = Constant()
  n5 = Add(A=MaterialFunctionCall(Hue Shift Percentage=Constant() {r=-0.05}, Texture=Desaturation(None=@n3, Fraction=@n4)) {materialFunction=&HueShift}.Result) {constB=0.3}
  n6 = Desaturation(
   None=MaterialFunctionCall!(Color=@n3, Luminosity=Constant() {r=2}) {materialFunction=/Engine/Functions/Engine_MaterialFunctions01/ImageAdjustment/Luminosity_And_Color}.Result
   Fraction=Constant()
  )
  n7 = MaterialFunctionCall(Color=@n3, Luminosity=Constant() {r=0.8}) {materialFunction=&Luminosity_And_Color}
  n8 = Constant3Vector()
  n9 = TextureCoordinate()
  n10 = TextureCoordinate()
  n11 = Constant2Vector() {g=0.5, r=0.5}
  n12 = Constant() {r=0.5}
  n13 = Constant()
  n14 = Power(
   Base=Multiply(A=SphereMask(A=@n10, B=@n11, Radius=@n12, Hardness=@n13), B=ScalarParameter() {defaultValue=5, parameterName=TwistAmount})
   Exp=ScalarParameter() {defaultValue=1, parameterName=TwistPower}
  )
  n15 = MaterialFunctionCall(Vector or UVs=Add(A=Rotator(Coordinate=@n9, Time=@n14), B=Constant2Vector() {g=-0.5, r=-0.5})) {materialFunction=&VectorToRadialValue}
  n16 = Constant2Vector() {g=1, r=2}
  n17 = Multiply(A=Time(), B=ScalarParameter() {defaultValue=1, parameterName=PanSpeed})
  n18 = TextureSample(UVs=Panner(Coordinate=Multiply(A=@n15.Radial Coordinates, B=@n16) {constA=1}, Time=@n17) {speedX=0.5, speedY=1.5}) {texture=/Game/Global/Textures/Default/Tileable_Patterns/T_Blowing_Smoke_Pattern2_Tile_DISP}
  n19 = Constant2Vector() {g=1, r=8}
  n20 = TextureSample(UVs=Panner(Coordinate=Multiply(A=@n15.Radial Coordinates, B=@n19) {constA=1}, Time=@n17) {speedX=0.3, speedY=0.8}) {texture=/Game/Global/Textures/Default/Noise/T_TileNoise_02}
  n21 = Add(A=Add(A=@n18.R, B=@n20.R))
  n22 = Power(
   Base=LinearInterpolate(A=Multiply(A=@n18.R, B=@n20.R), B=Multiply(A=@n18.R, B=@n21), Alpha=ScalarParameter() {defaultValue=1, parameterName=LerpClouds})
   Exp=Constant() {r=0.8}
  )
  n23 = Multiply(A=SphereMask(A=TextureCoordinate(), B=Constant2Vector() {g=0.5, r=0.5}, Radius=Constant() {r=0.4}, Hardness=Constant() {r=-8})) {constB=5}
  n24 = Power(Base=@n23)
  n25 = Subtract(A=OneMinus(None=Multiply(A=@n22) {constB=0.4}), B=Add(A=@n23, B=Multiply(A=@n24) {constB=2}))
  MaterialFunctionCall(Emissive=Multiply(A=MaterialFunctionCall(1=@n5, 2=@n6, 3=@n7.Result, 4=@n8, A=@n25) {materialFunction=&Lerp_Multiple_Float3}.Lerp 4 Inputs)) {materialFunction=&MF_FX_EmissiveAdjustments}
  Multiply(A=@n1)
  ScalarParameter() {parameterName=DDX}
  ScalarParameter() {parameterName=DDY}
  Subtract(A=@n1, B=@n2)
 ]
}
